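# Physics asset for the scarf on the OrinChai medium male player body, built from capsule bodies.
# A five-segment scarf tail chain hangs from a master bone on spine_05, with swing limits widening toward the tip and twist tightly limited.
# Capsules on pelvis, spine, upper arms, left thigh and one left skirt bone form the surrounding body shapes.
# Scarf masters, arms, thigh and skirt have free angular constraints, while spine links are limited.
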PhysicsAsset /BRCosmetics/Characters/Player/Male/Medium/Bodies/M_MED_OrinChai/Meshes/M_MED_OrinChai_Scraf_Physics
bodies {
 dyn_scarf_back_mstr [{shapeType=Capsule, length=0.505, radius=0.505, shapeName=dyn_scarf_back_mstr_capsule, rotation=<rot 0 -90 0>}]
 dyn_scarf_front_mstr [{shapeType=Capsule, length=0.505, radius=0.505, shapeName=dyn_scarf_front_mstr_capsule, rotation=<rot 0 -90 0>}]
 dyn_scarftail_01 [{shapeType=Capsule, length=5, radius=2.8, shapeName=dyn_scarftail_01_capsule, center=<3 -2 0>, rotation=<rot 90 0 0>}]
 dyn_scarftail_02 [{shapeType=Capsule, length=6, radius=3.7, shapeName=dyn_scarftail_02_capsule, center=<3 -2 0>, rotation=<rot 90 0 0>}]
 dyn_scarftail_03 [{shapeType=Capsule, length=6, radius=3.7, shapeName=dyn_scarftail_03_capsule, center=<3 -1.5 0>, rotation=<rot 90 0 0>}]
 dyn_scarftail_04 [{shapeType=Capsule, length=6.8808, radius=3.8, shapeName=dyn_scarftail_04_capsule, center=<3 0 0>, rotation=<rot 90 0 0>}]
 dyn_scarftail_05 [{shapeType=Capsule, length=8.6587, radius=4, shapeName=dyn_scarftail_05_capsule, center=<3 0 0>, rotation=<rot 90 0 0>}]
 dyn_scarftail_mstr [{shapeType=Capsule, length=0.505, radius=0.505, shapeName=dyn_scarftail_mstr_capsule, rotation=<rot 90 0 0>}]
 dyn_skirt_L_02 [{
   shapeType=Capsule
   length=15.1561
   radius=6.0143
   shapeName=dyn_skirt_L_02_capsule
   center=<5.5549 0 -0.3223>
   rotation=<rot 83.3857 -179.9999 -179.9999>
  }]
 pelvis [{shapeType=Capsule, length=15.0137, radius=13.2665, rotation=<rot -7.4797 -90 -1.0419>}]
 spine_01 [{shapeType=Capsule, length=11.6527, radius=11, center=<0 0.8773 -0.0017>, rotation=<rot 0.1118 -90 -0.08>}]
 spine_03 [{shapeType=Capsule, length=10.4418, radius=10, center=<0 0.8773 0.0006>, rotation=<rot -0.0389 -90 -0.0556>}]
 spine_04 [{shapeType=Capsule, length=10, radius=12, center=<3.4388 0.0167 -0.0118>, rotation=<rot 1.3033 -90 0.1908>}]
 spine_05 [{shapeType=Capsule, length=19, radius=11.2, center=<0 0.8754 0.0587>, rotation=<rot -3.8352 -90 0.1657>}]
 thigh_l [{shapeType=Capsule, length=15.5388, radius=10.7964, shapeName=thigh_l_capsule, center=<-17.517 0 0>, rotation=<rot 90 0 0>}]
 upperarm_l [{shapeType=Capsule, length=18, radius=7.25, shapeName=upperarm_l_capsule, center=<10 0 0>, rotation=<rot 90 0 0>}]
 upperarm_r [{shapeType=Capsule, length=15, radius=7, shapeName=upperarm_r_capsule, center=<-10 0 0>, rotation=<rot 90 0 0>}]
}
constraints [
 {bone1Name=spine_01, bone2Name=pelvis, swing1LimitDegrees=45, swing2LimitDegrees=45, twistLimitDegrees=45}
 {bone1Name=spine_03, bone2Name=spine_01, swing1LimitDegrees=45, swing2LimitDegrees=45, twistLimitDegrees=45}
 {bone1Name=spine_04, bone2Name=spine_03, swing1LimitDegrees=45, swing2LimitDegrees=45, twistLimitDegrees=45}
 {bone1Name=spine_05, bone2Name=spine_04, swing1LimitDegrees=45, swing2LimitDegrees=45, twistLimitDegrees=45}
 {bone1Name=dyn_scarftail_05, bone2Name=dyn_scarftail_04, swing1LimitDegrees=55, swing2LimitDegrees=55, twistLimitDegrees=2}
 {
  bone1Name=dyn_scarftail_mstr
  bone2Name=spine_05
  swing1LimitDegrees=45
  swing1Motion=Free
  swing2LimitDegrees=45
  swing2Motion=Free
  twistLimitDegrees=45
  twistMotion=Free
 }
 {bone1Name=dyn_scarftail_01, bone2Name=dyn_scarftail_mstr, swing1LimitDegrees=35, swing2LimitDegrees=30, twistLimitDegrees=2}
 {bone1Name=dyn_scarftail_02, bone2Name=dyn_scarftail_01, swing1LimitDegrees=40, swing2LimitDegrees=40, twistLimitDegrees=2}
 {bone1Name=dyn_scarftail_03, bone2Name=dyn_scarftail_02, swing1LimitDegrees=50, swing2LimitDegrees=50, twistLimitDegrees=2}
 {bone1Name=dyn_scarftail_04, bone2Name=dyn_scarftail_03, swing1LimitDegrees=55, swing2LimitDegrees=55, twistLimitDegrees=2}
 {
  bone1Name=upperarm_l
  bone2Name=spine_05
  swing1LimitDegrees=45
  swing1Motion=Free
  swing2LimitDegrees=45
  swing2Motion=Free
  twistLimitDegrees=45
  twistMotion=Free
 }
 {
  bone1Name=upperarm_r
  bone2Name=spine_05
  swing1LimitDegrees=45
  swing1Motion=Free
  swing2LimitDegrees=45
  swing2Motion=Free
  twistLimitDegrees=45
  twistMotion=Free
 }
 {
  bone1Name=dyn_scarf_front_mstr
  bone2Name=spine_05
  swing1LimitDegrees=45
  swing1Motion=Free
  swing2LimitDegrees=45
  swing2Motion=Free
  twistLimitDegrees=45
  twistMotion=Free
 }
 {
  bone1Name=dyn_scarf_back_mstr
  bone2Name=spine_05
  swing1LimitDegrees=45
  swing1Motion=Free
  swing2LimitDegrees=45
  swing2Motion=Free
  twistLimitDegrees=45
  twistMotion=Free
 }
 {
  bone1Name=dyn_skirt_L_02
  bone2Name=pelvis
  swing1LimitDegrees=45
  swing1Motion=Free
  swing2LimitDegrees=45
  swing2Motion=Free
  twistLimitDegrees=45
  twistMotion=Free
 }
 {
  bone1Name=thigh_l
  bone2Name=pelvis
  swing1LimitDegrees=45
  swing1Motion=Free
  swing2LimitDegrees=45
  swing2Motion=Free
  twistLimitDegrees=45
  twistMotion=Free
 }
]
