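# Animation Blueprint for the M_MED_OrinChai body, adding secondary motion to scarf and skirt bones.
# The parent mesh pose runs through deform and control rigs into a cache, feeding RigidBody and Trail branches joined by a bool blend.
# AnimDynamics, weighted constraints and bone offsets then shape scarf and skirt before a final Control Rig applies backbling-aware corrections.
# The event graph updates gravity override and movement-driven trail alpha, and detects backbling on customization load.
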AnimBlueprint /BRCosmetics/Characters/Player/Male/Medium/Bodies/M_MED_OrinChai/Meshes/M_MED_OrinChai_AnimBP
target_skeleton=&/M_MED_OrinChai_Skeleton
graphs [
 AnimGraph {nodes [
   n1 = Anim Dynamics - Bone: dyn_scarf_back(Alpha=1, ExternalForce="0,0,0") {
    bOverrideLinearDamping=true
    linearDampingOverride=1
    simulationSpace=BoneRelative
    alphaBoolBlend {blendOption=Linear}
    boundBone {boneName=dyn_scarf_back}
    componentAppliedLinearAccClamp=<10000 10000 10000>
    componentLinearAccScale=<1.14 1.14 1.14>
    componentLinearVelScale=<1.14 1.14 1.14>
    physicsBodyDefinitions [{
      boundBone {boneName=dyn_scarf_back}
      boxExtents=<70 70 70>
      constraintSetup {angularLimitsMax=<0 0 0>, angularLimitsMin=<0 0 0>, angularTarget=<0 0 0>, linearAxesMax=<0 0 10>}
      localJointOffset=<0 0 -10>
     }]
    relativeSpaceBone {boneName=root}
   }
   n2 = "Constraint - dyn_scarf_back (R)"(ComponentPose=@n1, Alpha=1, ConstraintWeights_0=1) {
    alphaBoolBlend {blendOption=Linear}
    boneToModify {boneName=dyn_scarf_back}
    constraintSetup [{transformType=Rotation, targetBone {boneName=dyn_scarf_back_mstr}}]
    constraintWeights [1]
   }
   n3 = Anim Dynamics - Bone: dyn_scarf_front(ComponentPose=@n2, Alpha=1, ExternalForce="0,0,0") {
    bOverrideLinearDamping=true
    linearDampingOverride=1
    simulationSpace=BoneRelative
    alphaBoolBlend {blendOption=Linear}
    boundBone {boneName=dyn_scarf_front}
    componentAppliedLinearAccClamp=<10000 10000 10000>
    componentLinearAccScale=<1.14 1.14 1.14>
    componentLinearVelScale=<1.14 1.14 1.14>
    physicsBodyDefinitions [{
      boundBone {boneName=dyn_scarf_front}
      boxExtents=<70 70 70>
      constraintSetup {angularLimitsMax=<0 0 0>, angularLimitsMin=<0 0 0>, angularTarget=<0 0 0>, linearAxesMax=<0 0 10>}
      localJointOffset=<0 0 -10>
     }]
    relativeSpaceBone {boneName=root}
   }
   "Constraint - dyn_scarf_front (R)"(ComponentPose=@n3, Alpha=1, ConstraintWeights_0=1) {
    alphaBoolBlend {blendOption=Linear}
    boneToModify {boneName=dyn_scarf_front}
    constraintSetup [{transformType=Rotation, targetBone {boneName=dyn_scarf_front_mstr}}]
    constraintWeights [1]
   }
   "Transform (Modify) Bone - Bone: dyn_scarftail_05"(Alpha=1, Translation="-4,0,0", Rotation="0,5,0", Scale="1,1,1") {
    rotationMode=BMM_Additive
    rotationSpace=BCS_BoneSpace
    translationMode=BMM_Additive
    translationSpace=BCS_BoneSpace
    alphaBoolBlend {blendOption=Linear}
    boneToModify {boneName=dyn_scarftail_05}
    rotation {yaw=5}
    translation {x=-4}
   }
   "Transform (Modify) Bone - Bone: dyn_scarftail_03"(Alpha=1, Translation="0,1,0", Rotation="0,5,0", Scale="1,1,1") {
    rotationMode=BMM_Additive
    rotationSpace=BCS_BoneSpace
    translationMode=BMM_Additive
    translationSpace=BCS_BoneSpace
    alphaBoolBlend {blendOption=Linear}
    boneToModify {boneName=dyn_scarftail_03}
    rotation {yaw=5}
    translation {y=1}
   }
   n4 = RigidBody!(
    ComponentPose=Local To Component(LocalPose=Use cached pose 'CtrlRigCache'())
    Alpha
    OverrideWorldGravity:Vector=Get GravityOverride()
    ExternalForce
   ) -> (Pose) {
    bClampLinearTranslationLimitToRefPose=true
    bOverrideWorldGravity=true
    overridePhysicsAsset=&/M_MED_OrinChai_Scraf_Physics
    simulationSpace=BaseBoneSpace
    alphaBoolBlend {blendOption=Linear}
    baseBoneRef {boneName=spine_01}
    componentAppliedLinearAccClamp=<3000 3000 3000>
    componentLinearAccScale=<0.3 0.2 0.2>
    componentLinearVelScale=<0.3 0.2 0.2>
    overrideWorldGravity {z=-980}
    simSpaceSettings {
     maxAngularAcceleration=200
     maxAngularVelocity=5
     maxLinearAcceleration=1000
     maxLinearVelocity=200
     worldAlpha=0.55
     externalLinearDragV=<1 1>
    }
   }
   n5 = Component To Local(ComponentPose=@n4)
   n6 = Trail controller - Bone: dyn_scarftail_05(ComponentPose=Local To Component(LocalPose=Use cached pose 'CtrlRigCache'()), Alpha=Get FwdMovement() -> (FwdMovement:double)) {
    bUsePlanarLimit=true
    chainLength=5
    debugLifeTime=0.1
    relaxationSpeedScale=1.5
    alphaScaleBiasClamp {bInterpResult=true}
    planarLimits [{drivingBone {boneName=spine_04}, planeTransform {location=<0 -13 -13>, rotation=<rot 0 10 -110>, scale=<1 1 1>}}]
    rotationLimits [
     {limitMax=<180 180 180>}
     {limitMax=<0 0 0>, limitMin=<0 0 0>}
     {limitMax=<0 0 0>, limitMin=<0 0 0>}
     {limitMax=<0 0 0>, limitMin=<0 0 0>}
     {limitMax=<180 180 180>}
    ]
    rotationOffsets [<0 0 0>, <0 0 0>, <0 0 0>, <0 0 0>, <0 0 0>]
    trailBone {boneName=dyn_scarftail_05}
    trailRelaxationSpeed {editorCurveData {keys [<at 0 15 RCIM_Linear>,<at 0.2082 8.812 RCIM_Linear>,<at 1 6 RCIM_Linear>]}}
   }
   n7 = Component To Local(ComponentPose=@n6)
   n8 = Anim Dynamics - Bone: dyn_scarf_back_01(
    ComponentPose=Local To Component(LocalPose=Blend Poses by bool(BlendPose_0=@n5, BlendPose_1=@n7) {blendType=Linear})
    Alpha=1
    ExternalForce="0,0,0"
   ) {
    angularDampingOverride=0.99
    angularSpringConstant=100
    bAngularSpring=true
    bOverrideAngularDamping=true
    simulationSpace=RootRelative
    alphaBoolBlend {blendOption=Linear}
    boundBone {boneName=dyn_scarf_back_01}
    componentAppliedLinearAccClamp=<10000 10000 10000>
    componentLinearAccScale=<1.5 1.5 1.5>
    componentLinearVelScale=<1.5 1.5 1.5>
    physicsBodyDefinitions [{
      boundBone {boneName=dyn_scarf_back_01}
      boxExtents=<25 25 25>
      constraintSetup {angularLimitsMax=<0 0 0>, angularLimitsMin=<0 -60 0>, angularTarget=<0 0 0>}
      localJointOffset=<1 0 0>
     }]
    relativeSpaceBone {boneName=pelvis}
   }
   n9 = Anim Dynamics - Bone: dyn_scarf_front_01(ComponentPose=@n8, Alpha=1, ExternalForce="0,0,0") {
    angularDampingOverride=1
    simulationSpace=RootRelative
    alphaBoolBlend {blendOption=Linear}
    boundBone {boneName=dyn_scarf_front_01}
    componentAppliedLinearAccClamp=<10000 10000 10000>
    componentLinearAccScale=<2 2 2>
    componentLinearVelScale=<2 2 2>
    physicsBodyDefinitions [{
      boundBone {boneName=dyn_scarf_front_01}
      boxExtents=<30 30 30>
      constraintSetup {angularLimitsMax=<0 0 0>, angularLimitsMin=<0 -55 0>, angularTarget=<0 0 0>}
      localJointOffset=<1 0 0>
     }]
    relativeSpaceBone {boneName=pelvis}
   }
   n10 = Anim Dynamics - Bone: dyn_skirt_L_02(ComponentPose=@n9, Alpha=1, ExternalForce="0,0,0") {
    angularDampingOverride=0.999
    angularSpringConstant=10000
    bAngularSpring=true
    bOverrideAngularDamping=true
    simulationSpace=RootRelative
    alphaBoolBlend {blendOption=Linear}
    boundBone {boneName=dyn_skirt_L_02}
    componentAppliedLinearAccClamp=<10000 10000 10000>
    componentLinearAccScale=<1 1 1>
    componentLinearVelScale=<1 1 1>
    physicsBodyDefinitions [{
      boundBone {boneName=dyn_skirt_L_02}
      boxExtents=<70 70 70>
      constraintSetup {angularLimitsMax=<5 30 0>, angularLimitsMin=<-5 -10 0>, angularTarget=<1 0 0>}
      localJointOffset=<10 0 0>
     }]
    relativeSpaceBone {boneName=pelvis}
   }
   n11 = Anim Dynamics - Bone: dyn_skirt_R_02(ComponentPose=@n10, Alpha=1, ExternalForce="0,0,0") {
    angularDampingOverride=0.999
    angularSpringConstant=10000
    bAngularSpring=true
    bOverrideAngularDamping=true
    simulationSpace=RootRelative
    alphaBoolBlend {blendOption=Linear}
    boundBone {boneName=dyn_skirt_R_02}
    componentAppliedLinearAccClamp=<10000 10000 10000>
    componentLinearAccScale=<1 1 1>
    componentLinearVelScale=<1 1 1>
    physicsBodyDefinitions [{
      boundBone {boneName=dyn_skirt_R_02}
      boxExtents=<70 70 70>
      constraintSetup {angularLimitsMax=<5 30 0>, angularLimitsMin=<-5 -10 0>, angularTarget=<1 0 0>}
      localJointOffset=<10 0 0>
     }]
    relativeSpaceBone {boneName=pelvis}
   }
   n12 = "Constraint - dyn_scarftail_02 (PP)"!(ComponentPose=@n11, Alpha=1, ConstraintWeights_0=0.8, ConstraintWeights_1:float=0.2) -> (Pose) {
    alphaBoolBlend {blendOption=Linear}
    boneToModify {boneName=dyn_scarftail_02}
    constraintSetup [{targetBone {boneName=dyn_scarftail_mstr}}, {targetBone {boneName=spine_01}}]
    constraintWeights [1,1]
   }
   n13 = "Transform (Modify) Bone - Bone: dyn_scarftail_02"(ComponentPose=@n12, Alpha=1, Translation="-4,2,-1", Rotation="0.016,12.721,-0.001", Scale="1,1,1") {
    rotationMode=BMM_Additive
    rotationSpace=BCS_BoneSpace
    translationMode=BMM_Additive
    translationSpace=BCS_BoneSpace
    alphaBoolBlend {blendOption=Linear}
    boneToModify {boneName=dyn_scarftail_02}
    rotation=<rot 0.016 12.721 -0.001>
    translation=<-4 2 -1>
   }
   n14 = "Constraint - dyn_scarftail_03 (PP)"!(ComponentPose=@n13, Alpha=1, ConstraintWeights_0=0.7, ConstraintWeights_1:float=0.3) -> (Pose) {
    alphaBoolBlend {blendOption=Linear}
    boneToModify {boneName=dyn_scarftail_03}
    constraintSetup [{targetBone {boneName=dyn_scarftail_02}}, {targetBone {boneName=spine_03}}]
    constraintWeights [1,1]
   }
   n15 = "Transform (Modify) Bone - Bone: dyn_scarftail_03"(ComponentPose=@n14, Alpha=1, Translation="-4,0,0", Rotation="0,12,0", Scale="1,1,1") {
    rotationMode=BMM_Additive
    rotationSpace=BCS_BoneSpace
    translationMode=BMM_Additive
    translationSpace=BCS_BoneSpace
    alphaBoolBlend {blendOption=Linear}
    boneToModify {boneName=dyn_scarftail_03}
    rotation {yaw=12}
    translation {x=-4}
   }
   n16 = "Constraint - dyn_scarftail_04 (PP)"!(ComponentPose=@n15, Alpha=1, ConstraintWeights_0=0.4, ConstraintWeights_1:float=0.6) -> (Pose) {
    alphaBoolBlend {blendOption=Linear}
    boneToModify {boneName=dyn_scarftail_04}
    constraintSetup [{targetBone {boneName=dyn_scarftail_03}}, {targetBone {boneName=spine_01}}]
    constraintWeights [1,1]
   }
   n17 = "Transform (Modify) Bone - Bone: dyn_scarftail_04"(ComponentPose=@n16, Alpha=1, Translation="-5,3,0", Rotation="0,5,0", Scale="1,1,1") {
    rotationMode=BMM_Additive
    rotationSpace=BCS_BoneSpace
    translationMode=BMM_Additive
    translationSpace=BCS_BoneSpace
    alphaBoolBlend {blendOption=Linear}
    boneToModify {boneName=dyn_scarftail_04}
    rotation {yaw=5}
    translation=<-5 3>
   }
   n18 = "Constraint - dyn_scarftail_05 (PP)"!(ComponentPose=@n17, Alpha=1, ConstraintWeights_0=0.2, ConstraintWeights_1:float=0.8) -> (Pose) {
    alphaBoolBlend {blendOption=Linear}
    boneToModify {boneName=dyn_scarftail_05}
    constraintSetup [{targetBone {boneName=dyn_scarftail_03}}, {targetBone {boneName=spine_01}}]
    constraintWeights [1,1]
   }
   n19 = "Transform (Modify) Bone - Bone: dyn_scarftail_05"(ComponentPose=@n18, Alpha=1, Translation="-10,7,0", Rotation="0,10,0", Scale="1,1,1") {
    rotationMode=BMM_Additive
    rotationSpace=BCS_BoneSpace
    translationMode=BMM_Additive
    translationSpace=BCS_BoneSpace
    alphaBoolBlend {blendOption=Linear}
    boneToModify {boneName=dyn_scarftail_05}
    rotation {yaw=10}
    translation=<-10 7>
   }
   n20 = Anim Dynamics - Bone: dyn_scarftail_05(ComponentPose=@n19, Alpha=1, ExternalForce="0,0,0") {
    angularDampingOverride=1
    bOverrideAngularDamping=true
    simulationSpace=RootRelative
    alphaBoolBlend {blendOption=Linear}
    boundBone {boneName=dyn_scarftail_05}
    componentAppliedLinearAccClamp=<1000 1000 1000>
    componentLinearAccScale=<0.8 0.8 0.8>
    componentLinearVelScale=<0.8 0.8 0.8>
    physicsBodyDefinitions [{
      boundBone {boneName=dyn_scarftail_05}
      boxExtents=<20 20 20>
      constraintSetup {angularLimitsMax=<0 15 0>, angularLimitsMin=<0 -15 -10>, angularTarget=<0 0 0>}
      localJointOffset=<6 0 0>
     }]
    relativeSpaceBone {boneName=pelvis}
   }
   n21 = "Transform (Modify) Bone - Bone: dyn_scarftail_04"(ComponentPose=@n20, Alpha=1, Translation="-3,1,0", Rotation="0,5,0", Scale="1,1,1") {
    rotationMode=BMM_Additive
    rotationSpace=BCS_BoneSpace
    translationMode=BMM_Additive
    translationSpace=BCS_BoneSpace
    alphaBoolBlend {blendOption=Linear}
    boneToModify {boneName=dyn_scarftail_04}
    rotation {yaw=5}
    translation=<-3 1>
   }
   n22 = "Constraint - dyn_scarftail_01 (T)"(ComponentPose=@n21, Alpha=1, ConstraintWeights_0=1) {
    alphaBoolBlend {blendOption=Linear}
    boneToModify {boneName=dyn_scarftail_01}
    constraintSetup [{transformType=Translation, targetBone {boneName=dyn_scarftail_mstr}}]
    constraintWeights [1]
   }
   n23 = Control Rig!(
    Source=Component To Local(ComponentPose=@n22)
    Alpha
    corrections:bool=true
    BackBlingCorrections:bool=Get IsWearingBackbling() -> (IsWearingBackbling:bool)
   ) -> (Pose) {
    alphaBoolBlend {blendOption=Linear}
    controlRigAssetReference {blueprintRigClass=&/M_MED_OrinChai_CtrlRig.M_MED_OrinChai_CtrlRig_C}
   }
   Output Pose(Result=@n23) {layerGroup=None}
   n24 = Control Rig!(Source=Copy Pose From Mesh() {bUseAttachedParent=true}, Alpha) -> (Pose) {
    alphaBoolBlend {blendOption=Linear}
    controlRigAssetReference {blueprintRigClass=&/M_MED_OrinChai_DeformRig.M_MED_OrinChai_DeformRig_C}
   }
   n25 = Control Rig!(Source=@n24, Alpha, corrections:bool=false, FrontEndCtrl:bool=false) -> (Pose) {
    alphaBoolBlend {blendOption=Linear}
    controlRigAssetReference {blueprintRigClass=&M_MED_OrinChai_CtrlRig_C}
   }
   Save cached pose 'CtrlRigCache'(Pose=@n25) {cacheName=CtrlRigCache}
   Get IsWearingCape() -> (IsWearingCape:bool)
  ]}
 EventGraph {nodes [
   Event BlueprintUpdateAnimation()
   > n1 = Sequence()
   > n2 = Gravity Override State Function DEPRECATED(CharacterPartAnimInstance=Self-Reference(), GameTimeInSeconds=GetGameTimeInSeconds(), GravityOverrideParams=Get Gravity Override Params())
   > Set GravityOverride(GravityOverride=@n2)
   n3 = Branch(execute=@n1.then_1, Condition=Get bIsInFrontEnd())
   > Set BwdMovement(BwdMovement:double=0) -> (Output_Get:double)
   n4 = PlayerMovement(execute=@n3.else, AnimInstance=Self-Reference(), XDotProductCoefficient=1, YDotProductCoefficient=1, ZDotProductCoefficient=1)
   n5 = float + float(A=float * float(A="Absolute (Float)"(A=@n4), B=0.5), B=MapRangeClamped(Value=@n4.YMovement, InRangeA=0, InRangeB=-1)) {is=title K2Node_CommutativeAssociativeBinaryOperator}
   n6 = MapRangeClamped(Value=@n5, InRangeA=0, InRangeB=1.1, OutRangeA=0.8, OutRangeB=0.1)
   > Set FwdMovement(FwdMovement:double=@n6) -> (Output_Get:double)
   Event OnCharacterCustomizationLoaded()
   n7 = CheckForBackbling()
   > n8 = Set IsWearingBackbling(IsWearingBackbling:bool=@n7) -> (Output_Get:bool)
   > Set IsWearingCape(IsWearingCape:bool=@n7.bIsBackblingACape) -> (Output_Get:bool)
   n9 = "Not Equal (Boolean)"(A=@n7.bIsBackblingACape, B=true) -> (ReturnValue=~)
   > Set ScarfEndAlpha(ScarfEndAlpha:double=SelectFloat(A=1, B=0, bPickA=AND Boolean(A=@n9, B=@n8))) -> (Output_Get:double)
   Set BwdMovement(BwdMovement=MapRangeClamped(Value=@n5, InRangeA=0, InRangeB=1.1))
   > Branch(Condition=Get bIsCrouching())
   Get Gravity Override Params()
   n10 = "Clamp (Float)"(Value=@n4.YMovement, Min=0, Max=1)
   Set CrouchScale(CrouchScale:double="Clamp (Float)"(Value=float - float(A=@n6, B=@n10), Min=0, Max=1)) -> (Output_Get:double)
   Set CrouchScale(CrouchScale=0)
  ]}
]
variables [
 BwdMovement {type=number}
 CrouchScale {type=number}
 FwdMovement {type=number}
 Gravity Override Params {
  type=GravityOverrideParamsStruct
  default {onGround_Falling {gravityOverride=<0 0 500>}, onGround_Moving {gravityOverride=<0 0 -800>}, skydive_GoingUp {gravityOverride=<0 -300 -1500>}}
 }
 GravityOverride {type=Vector}
 IsWearingBackbling {type=boolean}
 IsWearingCape {type=boolean}
 ScarfEndAlpha {type=number}
]
