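# Physics asset for the Orin Chai male medium body's FaceAcc mesh, defining capsule bodies and joints for its dynamic hair bones.
# Fifteen chains (a–o) hang from a head capsule, each starting at a small master body whose swing and twist are free.
# Chains a–j add two segments with narrow, symmetric swing limits.
# Chains k–o add one segment with asymmetric, wider swing, and all segments keep very tight twist limits.
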PhysicsAsset /BRCosmetics/Characters/Player/Male/Medium/Bodies/M_MED_OrinChai/Meshes/Parts/M_MED_OrinChai_FaceAcc_Physics
bodies {
 dyn_hair_a_01 [{shapeType=Capsule, length=2, radius=1, shapeName=dyn_hair_a_01_capsule, rotation=<rot 90 0 0>}]
 dyn_hair_a_02 [{shapeType=Capsule, length=2, radius=1, shapeName=dyn_hair_a_02_capsule, rotation=<rot 90 0 0>}]
 dyn_hair_a_mstr [{shapeType=Capsule, length=0.5, radius=0.5, shapeName=dyn_hair_a_mstr_capsule, rotation=<rot 90 0 0>}]
 dyn_hair_b_01 [{shapeType=Capsule, length=2, radius=1, shapeName=dyn_hair_b_01_capsule, rotation=<rot 90 0 0>}]
 dyn_hair_b_02 [{shapeType=Capsule, length=2, radius=1, shapeName=dyn_hair_b_02_capsule, rotation=<rot 90 0 0>}]
 dyn_hair_b_mstr [{shapeType=Capsule, length=0.5, radius=0.5, shapeName=dyn_hair_b_mstr_capsule, rotation=<rot 90 0 0>}]
 dyn_hair_c_01 [{shapeType=Capsule, length=2, radius=1, shapeName=dyn_hair_c_01_capsule, rotation=<rot 90 0 0>}]
 dyn_hair_c_02 [{shapeType=Capsule, length=2, radius=1, shapeName=dyn_hair_c_02_capsule, rotation=<rot 90 0 0>}]
 dyn_hair_c_mstr [{shapeType=Capsule, length=0.5, radius=0.5, shapeName=dyn_hair_c_mstr_capsule, rotation=<rot 90 0 0>}]
 dyn_hair_d_01 [{shapeType=Capsule, length=2, radius=1, shapeName=dyn_hair_d_01_capsule, rotation=<rot 90 0 0>}]
 dyn_hair_d_02 [{shapeType=Capsule, length=2, radius=1, shapeName=dyn_hair_d_02_capsule, rotation=<rot 90 0 0>}]
 dyn_hair_d_mstr [{shapeType=Capsule, length=0.5, radius=0.5, shapeName=dyn_hair_d_mstr_capsule, rotation=<rot 90 0 0>}]
 dyn_hair_e_01 [{shapeType=Capsule, length=2, radius=1, shapeName=dyn_hair_e_01_capsule, rotation=<rot 90 0 0>}]
 dyn_hair_e_02 [{shapeType=Capsule, length=2, radius=1, shapeName=dyn_hair_e_02_capsule, rotation=<rot 90 0 0>}]
 dyn_hair_e_mstr [{shapeType=Capsule, length=0.5, radius=0.5, shapeName=dyn_hair_e_mstr_capsule, rotation=<rot 90 0 0>}]
 dyn_hair_f_01 [{shapeType=Capsule, length=2, radius=1, shapeName=dyn_hair_f_01_capsule, rotation=<rot 90 0 0>}]
 dyn_hair_f_02 [{shapeType=Capsule, length=2, radius=1, shapeName=dyn_hair_f_02_capsule, rotation=<rot 90 0 0>}]
 dyn_hair_f_mstr [{shapeType=Capsule, length=0.5, radius=0.5, shapeName=dyn_hair_f_mstr_capsule, rotation=<rot 90 0 0>}]
 dyn_hair_g_01 [{shapeType=Capsule, length=2, radius=1, shapeName=dyn_hair_g_01_capsule, rotation=<rot 90 0 0>}]
 dyn_hair_g_02 [{shapeType=Capsule, length=2, radius=1, shapeName=dyn_hair_g_02_capsule, rotation=<rot 90 0 0>}]
 dyn_hair_g_mstr [{shapeType=Capsule, length=0.5, radius=0.5, shapeName=dyn_hair_g_mstr_capsule, rotation=<rot 90 0 0>}]
 dyn_hair_h_01 [{shapeType=Capsule, length=2, radius=1, shapeName=dyn_hair_h_01_capsule, rotation=<rot 90 0 0>}]
 dyn_hair_h_02 [{shapeType=Capsule, length=2, radius=1, shapeName=dyn_hair_h_02_capsule, rotation=<rot 90 0 0>}]
 dyn_hair_h_mstr [{shapeType=Capsule, length=0.5, radius=0.5, shapeName=dyn_hair_h_mstr_capsule, rotation=<rot 90 0 0>}]
 dyn_hair_i_01 [{shapeType=Capsule, length=2, radius=1, shapeName=dyn_hair_i_01_capsule, rotation=<rot 90 0 0>}]
 dyn_hair_i_02 [{shapeType=Capsule, length=2, radius=1, shapeName=dyn_hair_i_02_capsule, rotation=<rot 90 0 0>}]
 dyn_hair_i_mstr [{shapeType=Capsule, length=0.5, radius=0.5, shapeName=dyn_hair_i_mstr_capsule, rotation=<rot 90 0 0>}]
 dyn_hair_j_01 [{shapeType=Capsule, length=2, radius=1, shapeName=dyn_hair_j_01_capsule, rotation=<rot 90 0 0>}]
 dyn_hair_j_02 [{shapeType=Capsule, length=2, radius=1, shapeName=dyn_hair_j_02_capsule, rotation=<rot 90 0 0>}]
 dyn_hair_j_mstr [{shapeType=Capsule, length=0.5, radius=0.5, shapeName=dyn_hair_j_mstr_capsule, rotation=<rot 90 0 0>}]
 dyn_hair_k_01 [{shapeType=Capsule, length=2, radius=1.2, shapeName=dyn_hair_k_01_capsule, center=<0.5 0 -2>, rotation=<rot 90 0 0>}]
 dyn_hair_k_mstr [{shapeType=Capsule, length=0.5, radius=0.5, shapeName=dyn_hair_k_mstr_capsule, rotation=<rot 90 0 0>}]
 dyn_hair_l_01 [{shapeType=Capsule, length=2, radius=1, shapeName=dyn_hair_l_01_capsule, center=<0.5 0 0>, rotation=<rot 90 0 0>}]
 dyn_hair_l_mstr [{shapeType=Capsule, length=0.5, radius=0.5, shapeName=dyn_hair_l_mstr_capsule, rotation=<rot 90 0 0>}]
 dyn_hair_m_01 [{shapeType=Capsule, length=2, radius=1.5, shapeName=dyn_hair_m_01_capsule, center=<0 0 -1>, rotation=<rot 90 0 0>}]
 dyn_hair_m_mstr [{shapeType=Capsule, length=0.5, radius=0.5, shapeName=dyn_hair_m_mstr_capsule, rotation=<rot 90 0 0>}]
 dyn_hair_n_01 [{shapeType=Capsule, length=2, radius=2.5, shapeName=dyn_hair_n_01_capsule, center=<0 0 -2>, rotation=<rot 90 0 0>}]
 dyn_hair_n_mstr [{shapeType=Capsule, length=0.5, radius=0.5, shapeName=dyn_hair_n_mstr_capsule, rotation=<rot 90 0 0>}]
 dyn_hair_o_01 [{shapeType=Capsule, length=2, radius=3, shapeName=dyn_hair_o_01_capsule, center=<0 0 -1>, rotation=<rot 90 0 0>}]
 dyn_hair_o_mstr [{shapeType=Capsule, length=0.5, radius=0.5, shapeName=dyn_hair_o_mstr_capsule, rotation=<rot 90 0 0>}]
 head [{shapeType=Capsule, length=5.177, radius=5.0297, rotation=<rot 90 0 0>}]
}
constraints [
 {
  bone1Name=dyn_hair_a_mstr
  bone2Name=head
  swing1LimitDegrees=45
  swing1Motion=Free
  swing2LimitDegrees=45
  swing2Motion=Free
  twistLimitDegrees=45
  twistMotion=Free
 }
 {bone1Name=dyn_hair_a_01, bone2Name=dyn_hair_a_mstr, swing1LimitDegrees=10, swing2LimitDegrees=10, twistLimitDegrees=2}
 {bone1Name=dyn_hair_a_02, bone2Name=dyn_hair_a_01, swing1LimitDegrees=10, swing2LimitDegrees=10, twistLimitDegrees=2}
 {
  bone1Name=dyn_hair_b_mstr
  bone2Name=head
  swing1LimitDegrees=45
  swing1Motion=Free
  swing2LimitDegrees=45
  swing2Motion=Free
  twistLimitDegrees=45
  twistMotion=Free
 }
 {bone1Name=dyn_hair_b_01, bone2Name=dyn_hair_b_mstr, swing1LimitDegrees=10, swing2LimitDegrees=10, twistLimitDegrees=2}
 {bone1Name=dyn_hair_b_02, bone2Name=dyn_hair_b_01, swing1LimitDegrees=10, swing2LimitDegrees=10, twistLimitDegrees=2}
 {
  bone1Name=dyn_hair_c_mstr
  bone2Name=head
  swing1LimitDegrees=45
  swing1Motion=Free
  swing2LimitDegrees=45
  swing2Motion=Free
  twistLimitDegrees=45
  twistMotion=Free
 }
 {bone1Name=dyn_hair_c_01, bone2Name=dyn_hair_c_mstr, swing1LimitDegrees=13, swing2LimitDegrees=13, twistLimitDegrees=2}
 {bone1Name=dyn_hair_c_02, bone2Name=dyn_hair_c_01, swing1LimitDegrees=13, swing2LimitDegrees=13, twistLimitDegrees=2}
 {
  bone1Name=dyn_hair_d_mstr
  bone2Name=head
  swing1LimitDegrees=45
  swing1Motion=Free
  swing2LimitDegrees=45
  swing2Motion=Free
  twistLimitDegrees=45
  twistMotion=Free
 }
 {bone1Name=dyn_hair_d_01, bone2Name=dyn_hair_d_mstr, swing1LimitDegrees=10, swing2LimitDegrees=10, twistLimitDegrees=2}
 {bone1Name=dyn_hair_d_02, bone2Name=dyn_hair_d_01, swing1LimitDegrees=10, swing2LimitDegrees=10, twistLimitDegrees=2}
 {
  bone1Name=dyn_hair_e_mstr
  bone2Name=head
  swing1LimitDegrees=45
  swing1Motion=Free
  swing2LimitDegrees=45
  swing2Motion=Free
  twistLimitDegrees=45
  twistMotion=Free
 }
 {bone1Name=dyn_hair_e_01, bone2Name=dyn_hair_e_mstr, swing1LimitDegrees=10, swing2LimitDegrees=10, twistLimitDegrees=2}
 {bone1Name=dyn_hair_e_02, bone2Name=dyn_hair_e_01, swing1LimitDegrees=10, swing2LimitDegrees=10, twistLimitDegrees=2}
 {
  bone1Name=dyn_hair_f_mstr
  bone2Name=head
  swing1LimitDegrees=45
  swing1Motion=Free
  swing2LimitDegrees=45
  swing2Motion=Free
  twistLimitDegrees=45
  twistMotion=Free
 }
 {bone1Name=dyn_hair_f_01, bone2Name=dyn_hair_f_mstr, swing1LimitDegrees=15, swing2LimitDegrees=15, twistLimitDegrees=2}
 {bone1Name=dyn_hair_f_02, bone2Name=dyn_hair_f_01, swing1LimitDegrees=15, swing2LimitDegrees=15, twistLimitDegrees=2}
 {
  bone1Name=dyn_hair_g_mstr
  bone2Name=head
  swing1LimitDegrees=45
  swing1Motion=Free
  swing2LimitDegrees=45
  swing2Motion=Free
  twistLimitDegrees=45
  twistMotion=Free
 }
 {bone1Name=dyn_hair_g_01, bone2Name=dyn_hair_g_mstr, swing1LimitDegrees=15, swing2LimitDegrees=15, twistLimitDegrees=2}
 {bone1Name=dyn_hair_g_02, bone2Name=dyn_hair_g_01, swing1LimitDegrees=15, swing2LimitDegrees=15, twistLimitDegrees=2}
 {
  bone1Name=dyn_hair_h_mstr
  bone2Name=head
  swing1LimitDegrees=45
  swing1Motion=Free
  swing2LimitDegrees=45
  swing2Motion=Free
  twistLimitDegrees=45
  twistMotion=Free
 }
 {bone1Name=dyn_hair_h_01, bone2Name=dyn_hair_h_mstr, swing1LimitDegrees=15, swing2LimitDegrees=15, twistLimitDegrees=2}
 {bone1Name=dyn_hair_h_02, bone2Name=dyn_hair_h_01, swing1LimitDegrees=15, swing2LimitDegrees=15, twistLimitDegrees=2}
 {
  bone1Name=dyn_hair_i_mstr
  bone2Name=head
  swing1LimitDegrees=45
  swing1Motion=Free
  swing2LimitDegrees=45
  swing2Motion=Free
  twistLimitDegrees=45
  twistMotion=Free
 }
 {bone1Name=dyn_hair_i_01, bone2Name=dyn_hair_i_mstr, swing1LimitDegrees=10, swing2LimitDegrees=10, twistLimitDegrees=2}
 {bone1Name=dyn_hair_i_02, bone2Name=dyn_hair_i_01, swing1LimitDegrees=10, swing2LimitDegrees=10, twistLimitDegrees=2}
 {
  bone1Name=dyn_hair_j_mstr
  bone2Name=head
  swing1LimitDegrees=45
  swing1Motion=Free
  swing2LimitDegrees=45
  swing2Motion=Free
  twistLimitDegrees=45
  twistMotion=Free
 }
 {bone1Name=dyn_hair_j_01, bone2Name=dyn_hair_j_mstr, swing1LimitDegrees=10, swing2LimitDegrees=10, twistLimitDegrees=2}
 {bone1Name=dyn_hair_j_02, bone2Name=dyn_hair_j_01, swing1LimitDegrees=10, swing2LimitDegrees=10, twistLimitDegrees=2}
 {
  bone1Name=dyn_hair_k_mstr
  bone2Name=head
  swing1LimitDegrees=45
  swing1Motion=Free
  swing2LimitDegrees=45
  swing2Motion=Free
  twistLimitDegrees=45
  twistMotion=Free
 }
 {bone1Name=dyn_hair_k_01, bone2Name=dyn_hair_k_mstr, swing1LimitDegrees=20, swing2LimitDegrees=40, twistLimitDegrees=2}
 {
  bone1Name=dyn_hair_l_mstr
  bone2Name=head
  swing1LimitDegrees=45
  swing1Motion=Free
  swing2LimitDegrees=45
  swing2Motion=Free
  twistLimitDegrees=45
  twistMotion=Free
 }
 {bone1Name=dyn_hair_l_01, bone2Name=dyn_hair_l_mstr, swing1LimitDegrees=15, swing2LimitDegrees=55, twistLimitDegrees=2}
 {
  bone1Name=dyn_hair_m_mstr
  bone2Name=head
  swing1LimitDegrees=45
  swing1Motion=Free
  swing2LimitDegrees=45
  swing2Motion=Free
  twistLimitDegrees=45
  twistMotion=Free
 }
 {bone1Name=dyn_hair_m_01, bone2Name=dyn_hair_m_mstr, swing1LimitDegrees=13, swing2LimitDegrees=55, twistLimitDegrees=2}
 {
  bone1Name=dyn_hair_n_mstr
  bone2Name=head
  swing1LimitDegrees=45
  swing1Motion=Free
  swing2LimitDegrees=45
  swing2Motion=Free
  twistLimitDegrees=45
  twistMotion=Free
 }
 {bone1Name=dyn_hair_n_01, bone2Name=dyn_hair_n_mstr, swing1LimitDegrees=15, swing2LimitDegrees=35, twistLimitDegrees=2}
 {
  bone1Name=dyn_hair_o_mstr
  bone2Name=head
  swing1LimitDegrees=45
  swing1Motion=Free
  swing2LimitDegrees=45
  swing2Motion=Free
  twistLimitDegrees=45
  twistMotion=Free
 }
 {bone1Name=dyn_hair_o_01, bone2Name=dyn_hair_o_mstr, swing1LimitDegrees=15, swing2LimitDegrees=55, twistLimitDegrees=2}
]
